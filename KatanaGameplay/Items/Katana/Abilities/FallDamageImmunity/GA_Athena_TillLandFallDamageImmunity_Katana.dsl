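; Katana gameplay ability that grants fall damage immunity, by its name lasting until the player lands.
; On activation it runs the parent ability only when the EnableFallDamageImmunity scalable float evaluates true, otherwise it ends immediately.
; CanApplyHolster always returns false, and the failed passive activation event is left empty.
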(event Ability|EventActivateAbility
  (if (ScalableFloat|GetValueasBool (Variables|Default|GetEnableFallDamageImmunity))
    (|CallParentFunction)
    (else
      (Ability|EndAbility))))

(event Ability|EventFailedToActivatePassiveAbility
  (params (ActorInfo :type Gameplay%20Ability%20Actor%20Info%20Structure)))

(fn CanApplyHolster
  (params (HolsterId :type Name%20%28by%20ref%29 :default "None"))
  (return false))
